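# A Sound Cue for Jones voice line 128 in the LimeRock event, part of the S37 VO set.
# A single Dialogue Player node plays the matching Dialogue Wave, with the LimeRock dialogue voice as speaker and no targets.
# It uses the LimeRock VO attenuation asset and the Fort Events VO sound class, with a raised volume multiplier.
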SoundCue /LimeRock/Assets/Audio/VO/Jones/VO_S37_Jones_LimerockEvent_128_Dialogue_Cue
root=DialoguePlayer_0
nodes [{
  id=DialoguePlayer_0
  properties {dialogueWaveParameter {
    dialogueWave=/LimeRock/Assets/Audio/VO/Jones/Waves/VO_S37_Jones_LimerockEvent_128_Dialogue
    context {speaker=/LimeRock/Assets/Audio/VO/LimeRock_DV, targets []}
   }}
 }]
properties {
 attenuationSettings=/LimeRock/Assets/Audio/Settings/ATT_LimeRock_VO
 maxDistance=12919.541
 soundClassObject=/Game/Sounds/FortSoundClassesAndModes/Fort_Events_VO
 volumeMultiplier=1.5
}
